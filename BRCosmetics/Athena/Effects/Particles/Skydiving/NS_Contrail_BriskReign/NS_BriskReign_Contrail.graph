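# Niagara system for the BriskReign skydiving contrail cosmetic, an ice-themed effect attached to the player's skeletal mesh.
# Crystal meshes follow a socket with catch-up offsets, wobble and smoothed, velocity-based banking from a borrowed RedPepper orientation module.
# Hand_Glow, IceDust and Snowflakes add socket-anchored glows, dust and curl-noise flakes, alongside a Swirls emitter.
# System Dive blends motion, spawn rates and color, with front-end and time-of-day modifiers adjusting placement and brightness.
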NiagaraSystem /BRCosmetics/Athena/Effects/Particles/Skydiving/NS_Contrail_BriskReign/NS_BriskReign_Contrail
dependencies {
 usedDataInterfaces [
  &NiagaraDataInterfaceCurve
  &NiagaraDataInterfaceMeshRendererInfo
  &NiagaraDataInterfacePlatformSet
  &NiagaraDataInterfaceSpriteRendererInfo
  &NiagaraDataInterfaceArrayFloat
  &NiagaraDataInterfaceSkeletalMesh
 ]
 usedDynamicInputs [
  &SimulationPosition
  &UniformRangedVector2D
  &FN_FrontEndVector2Modifier
  &MakeFloatFromVector
  &FNTimeOfDayColor
 ]
 usedModules [
  &self:SystemSpawnScript.SystemScriptSource.SystemScriptGraph.NiagaraNodeAssignment_2.SetVariables_#1
  &SystemState
  &self:UpdatePreviousRotation
  /Game/Effects/Niagara/Modules/Skydiving/FN_DivingAttributes/v5/FN_DivingAttributes
  /Game/Effects/Niagara/Modules/Contrails/FN_Contrail_Setup
  &self:SystemSpawnScript.SystemScriptSource.SystemScriptGraph.NiagaraNodeAssignment_0.SetVariables_#2
  &self:SystemSpawnScript.SystemScriptSource.SystemScriptGraph.NiagaraNodeAssignment_4.SetVariables_#3
  &self:SystemSpawnScript.SystemScriptSource.SystemScriptGraph.NiagaraNodeAssignment_5.SetVariables_#4
  &FN_Integrator
  &self:SystemSpawnScript.SystemScriptSource.SystemScriptGraph.NiagaraNodeAssignment_6.SetVariables_#5
  &self:SystemSpawnScript.SystemScriptSource.SystemScriptGraph.NiagaraNodeAssignment_1.SetVariables_#6
  &self:SystemSpawnScript.SystemScriptSource.SystemScriptGraph.NiagaraNodeAssignment_3.SetVariables_#7
  &self:SystemSpawnScript.SystemScriptSource.SystemScriptGraph.NiagaraNodeAssignment_7.SetVariables_#8
  &self:SystemSpawnScript.SystemScriptSource.SystemScriptGraph.NiagaraNodeAssignment_8.SetVariables_#9
 ]
}
emitters {
 Crystal {
  emitterUpdateScript [
   &EmitterState()
   &SpawnBurst_Instantaneous(Spawn Count=3, Use Spawn Probability=false, UseLoopCountLimit=true)
   &InterpolateOverTime(Rate Of Change=50, Target Vector Value:Vector3f=$Engine.Owner.Velocity, Variable Type=::NewEnumerator2)
  ]
  particleSpawnScript [
   &FN_SetSpawnGroup(Number of Spawn Groups=3)
   &InitializeParticle(
    Lifetime=2
    Mesh Scale Mode=::NewEnumerator3
    Mesh Uniform Scale=0.8
    Sprite Size=<71.0167 113.1586>
    Sprite Size Mode=::NewEnumerator4
    UsePositionOffset=true
    Write Material Random=true
   )
   /SetVariables(Particles.VisibilityTag=&SpawnGroup()) {isSetParameters=true, module=SetVariables_#10}
   /SetVariables(Particles.RandomInputPos:NiagaraFloat=&RandomRangeFloat(Minimum=0, Maximum=0)) {isSetParameters=true, module=SetVariables_#11}
   &CameraOffset(Camera Offset Amount=&MergeHelper_FloatSwitch(0=-5, 1=0, 2=0, 3=0, 4=0, Index=&SpawnGroup(), Num=::NewEnumerator3))
   /SetVariables(Particles.GlowSpriteDynamic:Vector4f=<1 20 1 1>) {isSetParameters=true}
  ]
  particleUpdateScript [
   &ParticleState(Kill Particles When Lifetime Has Elapsed=false, Let Infinitely Lived Particles Die When Emitter Deactivates=true)
   /SetVariables(Particles.Ambient:Vector3f=$System.Ambient) {isSetParameters=true, module=SetVariables_#12}
   &InterpolateOverTime(
    Rate Of Change=3
    Target Float Value:NiagaraFloat=&MakeCustomFloatFromBool(Bool=$System.Timeline001.IsRewinding, True Float=0, False Float=1)
   ) {module=InterpolateOverTime004}
   /SetVariables(Particles.CatchupVector:Vector3f=&Multiply_VectorByFloat(
     Vector=&Multiply_VectorByFloat(
      Vector=&NormalizeVector(Vector To Normalize=$System.SkydiveVelocity, Safe Normalize=true, Safe Fallback Threshold=0, Safe Fallback Vector=<0 0 0.001>)
      Float=-75
     )
     Float=&Lerp_Float(
      A=&FloatFromCurve(
       FloatCurve=DataInterface() {propertyValues {Curve {keys [
           {interpMode=RCIM_Cubic, tangentMode=RCTM_User, time=0, value=0}
           {interpMode=RCIM_Cubic, tangentMode=RCTM_User, time=0.35, value=-0.35}
           {tangentMode=RCTM_User, time=1, value=0}
          ]}}}
       CurveIndex=$System.Timeline001.NormalizedTime
      )
      B=&FloatFromCurve(
       FloatCurve=DataInterface() {propertyValues {Curve {keys [
           {arriveTangent=3.3827, arriveTangentWeight=1.1106, interpMode=RCIM_Cubic, leaveTangentWeight=1.1106, time=0, value=0}
           {arriveTangentWeight=0.5241, interpMode=RCIM_Cubic, leaveTangentWeight=0.1667, time=0.5, value=0.65}
           <at 1 0 RCIM_Cubic>
          ]}}}
       CurveIndex=$System.Timeline001.NormalizedTime
      )
      Alpha=$Particles.InterpolateOverTime004.InterpolateOverTime_Float.Moving Average
     )
    )) {isSetParameters=true, module=SetVariables_#13}
   &SkeletalMeshLocation(
    Bone / Socket Index=4
    Bone Sampling Mode=::NewEnumerator4
    MaskSpawnGroup=false
    Orientation Sampling=::NewEnumerator1
    Sampled Position Offset=&Add_Vector(
     A=$Particles.CatchupVector
     B=/Niagara/DynamicInputs/Vectors/RotateVector(
      Vector To Rotate:Vector3f=&Add_Vector(
       A=$Particles.Ambient
       B=&Add_Vector(
        A=&Add_Vector(
         A=$Particles.ShapeLocation.ShapeVector
         B=&MergeHelper_VectorSwitch(
          0=&Lerp_Vector(
           A=<0 40.5 66>
           B=&FN_FrontEndVector3Modifier(Gameplay=<0 82 34>, Front End=<0 82 30>)
           Alpha=$System.InterpolateOverTime.InterpolateOverTime_Float.Moving Average
          )
          1=&Lerp_Vector(
           A=<0 40 65>
           B=&FN_FrontEndVector3Modifier(Gameplay=<0 82 32>, Front End=<0 82 32>)
           Alpha=$System.InterpolateOverTime.InterpolateOverTime_Float.Moving Average
          )
          2=<0 0 0>
          3=<0 0 0>
          4=<0 0 0>
          Index=&SpawnGroup()
          Num=::NewEnumerator3
         )
        )
        B=&Add_Vector(A=$System.OverallTargetOffset, B=$System.FEOffset)
       )
      )
      Yaw:NiagaraFloat=&Waveform(
       "Waveform [1]"=::NewEnumerator12
       "[1] Input Position"=$Particles.RandomInputPos
       "[1] Frequency"=1.25
       "[1] Amplitude Min/Max"=<7 -7>
       "[1] Scale"=0
      )
      Pitch:NiagaraFloat=0
      Roll:NiagaraFloat=0
      Rotation Rate:NiagaraFloat=2
     )
    )
    Skeletal Mesh=$System.Skeletal Mesh
    Use Sampled Mesh Attribute as Ribbon ID=false
   )
   &ScaleMeshSize(Scale Factor=&VectorFromFloat(Value=&FloatFromCurve(
      FloatCurve=DataInterface() {propertyValues {Curve {keys [{arriveTangent=1.0186, interpMode=RCIM_Cubic, leaveTangent=1.0186, tangentMode=RCTM_User, time=0, value=0}, <at 1 1 RCIM_Cubic>]}}}
      CurveIndex=$Emitter.Age
     )))
   /SetVariables(Particles.Wobble:Quat4f=&MultiplyQuaternion(
     Quaternion A=&FindQuatBetweenTwoVectors(
      From Vector=&MakeVector(
       X=&Sine(
        Normalized Angle=&Add_Float(A=$System.FN_Integrator.Value, B=&RandomRangeFloat(Minimum=0.5, Maximum=0.5))
        Period=&MakeCustomFloatFromBool(Bool=&RandomBool(), True Float=1, False Float=-1)
        Scale=0.2
        Bias=0
       )
       Y=&Cosine(
        Normalized Angle=&Add_Float(A=$System.FN_Integrator.Value, B=&RandomRangeFloat(Minimum=0, Maximum=1))
        Period=&MakeCustomFloatFromBool(Bool=&RandomBool(), True Float=1, False Float=-1)
        Scale=0.2
        Bias=0
       )
       Z=&RandomRangeFloat(Minimum=6, Maximum=6)
      )
      To Vector=<0 0 1>
     )
     Quaternion B=$Particles.Initial.MeshOrientation
    )) {isSetParameters=true, module=SetVariables_#14}
   &UpdateMeshOrientation(
    Bank Rate=900000000
    Banking Angle Min / Max:Vector2f=<0.1 -0.1>
    Enabled:NiagaraBool=&SetBoolByFloatComparison(A=&FN_FrontEndFloatModifier(Gameplay=0, Front End=1), B=1, Comparison Type=::NewEnumerator4)
    "Foward Vector (Mesh Space)":Vector3f=&Lerp_Vector(A=<0.5 0 1>, B=<0.5 0 0>, Alpha=$System.InterpolateOverTime.InterpolateOverTime_Float.Moving Average)
    Orientation Method=::NewEnumerator3
    Particle Velocity:Vector3f=&Multiply_VectorByFloat(
     Vector=&Add_Vector(
      A=&ReturnMeshOrientationAxis(Mesh Orientation Axis=::YAxis, Mesh Orientation Quat=$Particles.SkeletalMeshLocation.SampledRotation)
      B=&Lerp_Vector(A=<0 0 -2>, B=<0 0 -0.5>, Alpha=$System.Dive)
     )
     Float=500000
    )
    Rotation Decay Rate In Degrees=900000000
    Up Vector:Vector3f=&Lerp_Vector(
     A=&ReturnMeshOrientationAxis(Mesh Orientation Axis=::YAxis, Mesh Orientation Quat=$Particles.SkeletalMeshLocation.SampledRotation)
     B=&Multiply_VectorByFloat(
      Vector=&ReturnMeshOrientationAxis(Mesh Orientation Axis=::ZAxis, Mesh Orientation Quat=$Particles.SkeletalMeshLocation.SampledRotation)
      Float=1
     )
     Alpha=&FloatFromCurve(FloatCurve=DataInterface() {propertyValues {Curve {keys [<at 0 0.5>,<at 0.25 0.75>]}}}, CurveIndex=$System.Dive)
    )
    Upright Roll Correction Speed=900000000
    Velocity Orientation Convergence Rate=900000000
   ) {module=UpdateMeshOrientation001}
   /BRCosmetics/Athena/Effects/Particles/Skydiving/NS_Contrail_RedPepper/UpdateMeshOrientation_RedPepper_TEMP(
    Bank Rate:NiagaraFloat=&Lerp_Float(
     A=&Lerp_Float(
      A=1
      B=30
      Alpha=&Waveform(
       "Waveform [1]"=::NewEnumerator12
       "[1] Input Position"=$Particles.RandomInputPos
       "[1] Phase Scale"=50000
       "[1] Frequency"=0.5
       "[1] Amplitude Min/Max"=<0 1>
      )
     )
     B=900
     Alpha=$System.Dive
    )
    Banking Angle Min / Max:Vector2f=&Lerp_Vector2(A=<-10 10>, B=<-4 4>, Alpha=$System.Dive)
    Enabled:NiagaraBool=&SetBoolByFloatComparison(A=&FN_FrontEndFloatModifier(Gameplay=0, Front End=1), B=0, Comparison Type=::NewEnumerator4)
    "Foward Vector (Mesh Space)":Vector3f=&Lerp_Vector(
     A=<0 0 -1>
     B=&MergeHelper_VectorSwitch(0=<2 0 1>, 1=<-2 0 -1>, 2=<0 0 0>, 3=<0 0 0>, 4=<0 0 0>, Index=&SpawnGroup(), Num=::NewEnumerator3)
     Alpha=$System.Dive
    )
    Orientation Method:ENiagara_UpdateMeshOrientationMode=::NewEnumerator3
    Particle Velocity:Vector3f=$Emitter.InterpolateOverTime.InterpolateOverTime_V3_.Moving Average
    "Right Vector (Mesh Space)":Vector3f=&Lerp_Vector(
     A=<0 -1 0>
     B=&MergeHelper_VectorSwitch(0=<0 1 0>, 1=<0 -1 0>, 2=<0 0 0>, 3=<0 0 0>, 4=<0 0 0>, Index=&SpawnGroup(), Num=::NewEnumerator3)
     Alpha=$System.Dive
    )
    Rotation Decay Rate In Degrees:NiagaraFloat=&Lerp_Float(A=150, B=500, Alpha=$System.Dive)
    Up Vector:Vector3f=&Lerp_Vector(
     A=&ReturnMeshOrientationAxis(Mesh Orientation Axis=::YAxis, Mesh Orientation Quat=$Particles.SkeletalMeshLocation.SampledRotation)
     B=&Multiply_VectorByFloat(
      Vector=&ReturnMeshOrientationAxis(Mesh Orientation Axis=::ZAxis, Mesh Orientation Quat=$Particles.SkeletalMeshLocation.SampledRotation)
      Float=1
     )
     Alpha=$System.Dive
    )
    Upright Roll Correction Speed:NiagaraFloat=&Lerp_Float(A=25, B=90000, Alpha=$System.Dive)
    Velocity Orientation Convergence Rate:NiagaraFloat=&Lerp_Float(
     A=&Lerp_Float(
      A=3
      B=12
      Alpha=&Waveform("Waveform [1]"=::NewEnumerator12, "[1] Input Position"=$Particles.RandomInputPos, "[1] Frequency"=5, "[1] Amplitude Min/Max"=<0 1>)
     )
     B=&Lerp_Float(
      A=20
      B=40
      Alpha=&Waveform("Waveform [1]"=::NewEnumerator12, "[1] Input Position"=$Particles.RandomInputPos, "[1] Frequency"=5, "[1] Amplitude Min/Max"=<0 1>)
     )
     Alpha=$System.Dive
    )
   ) {module=UpdateMeshOrientation_RedPepper_TEMP001}
   &InterpolateOverTime(
    Rate Of Change:NiagaraFloat=&Lerp_Float(
     A=4
     B=2
     Alpha=&FloatFromCurve(
      FloatCurve=DataInterface() {propertyValues {Curve {keys [
          {interpMode=RCIM_Cubic, leaveTangent=0.0851, tangentMode=RCTM_Break, time=0, value=0}
          {interpMode=RCIM_Cubic, tangentMode=RCTM_Break, time=0.222, value=1}
         ]}}}
      CurveIndex=$System.Dive
     )
    )
    Target Quat Value:Quat4f=$Particles.MeshOrientation
    Variable Type=::NewEnumerator5
   ) {module=InterpolateOverTime002}
   /SetVariables(Particles.New Orient:Quat4f=$StackContext.InterpolateOverTime002.InterpolateOverTime_Quaternion.Moving Average) {isSetParameters=true, module=SetVariables_#15}
   /SetVariables(Particles.MeshOrientation=&MultiplyQuaternion(Quaternion A=$StackContext.Wobble, Quaternion B=$Particles.New Orient)) {isSetParameters=true, module=SetVariables_#16}
   &ScaleColor(
    Scale RGB=&MakeVectorFromLinearColorRGB(Color=&LerpLinearColors(StartColor=<rgba 0 1.3565 2 1>, EndColor=<rgba 4 0.479 0 1>, LerpFactor=$System.InterpolateOverTime.InterpolateOverTime_Float.Moving Average))
    ScaleA=true
    ScaleRGB=true
   )
   &DynamicMaterialParameters(
    Index 0 Param 1=1
    Index 0 Param 2=0.5
    Index 0 Param 3=&RandomRangeFloat(Minimum=0.1, Maximum=1)
    Index 0 Param 4=&RandomRangeFloat(Minimum=0.75, Maximum=1)
    Module.Param0WriteEnabled=true
    Module.Param1WriteEnabled=true
    Module.Param2WriteEnabled=true
    Module.Param3WriteEnabled=true
   )
   /SetVariables(
    Particles.Decal Color:LinearColor=&LerpLinearColors(
     StartColor=<rgba 0 1.3565 2 3>
     EndColor=<rgba 6 0.9039 0 3>
     LerpFactor=$System.InterpolateOverTime.InterpolateOverTime_Float.Moving Average
    )
    Particles.Decal Offset=15
    Particles.Decal Position:NiagaraPosition=&AddVectorToPosition(
     Position=$Particles.Position
     Vector=&RotateVectorByQuaternion(Vector To Rotate=<-10 0 15>, Quaternion=$Particles.MeshOrientation)
    )
   ) {isSetParameters=true, module=SetVariables_#17}
   /SetVariables(
    Particles.ScpeedConeScale:Vector3f=<2 2 5>
    Particles.SpeedConeColor:LinearColor=&MakeLinearColorFromVectorAndFloat(
     "Vector (RGB)"=&MakeVectorFromLinearColorRGB(Color=<rgba 0.7552 0.8886 1 1>)
     "Float (Alpha)"=&Multiply_Float(
      A=&FloatFromCurve(FloatCurve=DataInterface() {propertyValues {Curve {keys [<at 0.6563 0 RCIM_Cubic>,<at 1 1>]}}}, CurveIndex=$System.Dive, Scale Curve=3)
      B=&FN_FrontEndFloatModifier(Gameplay=1, Front End=0.5)
     )
    )
   ) {isSetParameters=true}
   /SetVariables(Particles.BurgerSize:Vector3f=<0.06 0.06 0.06>) {isSetParameters=true}
   /SetVariables(
    Particles.CrystalColor:LinearColor=&Multiply_LinearColorByFloat(
     Linear Color=&LerpLinearColors(
      StartColor=<rgba 1.375 1.5946 3 1.3>
      EndColor=<rgba 1.375 1.9999 6 1.3>
      LerpFactor=&FloatFromCurve(FloatCurve=DataInterface() {propertyValues {Curve {keys [<at 0.6 0>,<at 1 1>]}}}, CurveIndex=$System.Dive)
     )
     Float=&Multiply_Float(A=$System.TODFloat, B=&FN_FrontEndFloatModifier(Gameplay=1, Front End=0.85))
    )
    Particles.CrystalrSize:Vector3f=&Multiply_Vector(A=<0.832 1 1>, B=&VectorFromFloat(Value=1.05))
   ) {isSetParameters=true}
   /SetVariables(
    Particles.GlowMeshColor:LinearColor=&Multiply_LinearColorByFloat(Linear Color=<rgba 0.2431 0.4526 1 0.4609>, Float=&Lerp_Float(A=0.6, B=1.2, Alpha=$System.Dive))
    Particles.GlowMeshSize:Vector3f=<2 2 3.539>
   ) {isSetParameters=true}
   /SetVariables(Particles.GlowSpriteColor:LinearColor=&Multiply_LinearColorByFloat(
     Linear Color=&Multiply_LinearColorByFloat(
      Linear Color=&LerpLinearColors(StartColor=<rgba 1 1.9046 3 0.217>, EndColor=<rgba 1 1.9046 3 0.5>, LerpFactor=$System.Dive)
      Float=&FN_FrontEndFloatModifier(Gameplay=1, Front End=0.6)
     )
     Float=&Multiply_Float(A=$System.TODFloat, B=0.525)
    )) {isSetParameters=true}
  ]
  rendererClasses [&NiagaraMeshRendererProperties, &NiagaraSpriteRendererProperties]
  renderers [
   {rendererClass=&NiagaraMeshRendererProperties, rendererIndex=0}
   {rendererClass=&NiagaraMeshRendererProperties, rendererIndex=1}
   {rendererClass=&NiagaraMeshRendererProperties, rendererIndex=2}
   {rendererClass=&NiagaraMeshRendererProperties, rendererIndex=3}
   {rendererClass=&NiagaraSpriteRendererProperties, rendererIndex=4}
  ]
 }
 Hand_Glow {
  emitterUpdateScript [
   &EmitterState(
    Inactive Response=::NewEnumerator1
    Life Cycle Mode=::NewEnumerator1
    Loop Behavior=::NewEnumerator1
    Loop Duration Mode=::NewEnumerator1
    UseLoopDelay=false
   )
   &SpawnBurst_Instantaneous(Spawn Count=2, Use Spawn Probability=false)
  ]
  particleSpawnScript [
   &FN_SetSpawnGroup(Number of Spawn Groups=2)
   &InitializeParticle(
    Color=<rgba 0.7375 0.8234 1 0.3>
    Lifetime=2
    Mass Mode=::NewEnumerator1
    Position Offset=<0 0 100>
    Sprite Size Mode=::NewEnumerator3
    Uniform Sprite Size=120
    UsePositionOffset=false
    Write Material Random=true
   ) {module=InitializeParticle001}
  ]
  particleUpdateScript [
   &ParticleState(Kill Particles When Lifetime Has Elapsed=false, Let Infinitely Lived Particles Die When Emitter Deactivates=true)
   &SkeletalMeshLocation(
    Bone / Socket Index=&SpawnGroup()
    Bone Sampling Mode=::NewEnumerator4
    MaskSpawnGroup=false
    Sampled Position Offset=&MergeHelper_VectorSwitch(0=<6 0 0>, 1=<-6 0 0>, 2=?, 3=?, Index=&SpawnGroup(), Num=::NewEnumerator0)
    Skeletal Mesh=$System.Skeletal Mesh
   ) {module=SkeletalMeshLocation001}
   &ScaleSpriteSize(
    Uniform Curve Scale=&Sine(Normalized Angle=&Add_Float(A=$Emitter.Age, B=0), Period=0.8, Scale=0.03, Bias=0.65)
    Uniform Curve Sprite Scale=DataInterface() {propertyValues {Curve {keys [{tangentMode=RCTM_User, time=0, value=1}, {tangentMode=RCTM_User, time=1, value=1}]}}}
   )
   &ScaleColor(
    Scale Alpha=&FloatFromCurve(
     FloatCurve=DataInterface() {propertyValues {Curve {keys [<at 0 0.7 RCIM_Cubic>,<at 0.5 1 RCIM_Cubic>,<at 1 0.7 RCIM_Cubic>]}}}
     CurveIndex=Linked() {linkedVariable {name=NPC.FortniteNPC.TimeOfDay_0-1, type=NiagaraFloat}}
     Scale Curve=&FN_FrontEndFloatModifier(Gameplay=1, Front End=0.7)
    )
    ScaleA=true
    ScaleRGB=true
   )
  ]
  rendererClasses [&NiagaraSpriteRendererProperties]
  renderers [{rendererClass=&NiagaraSpriteRendererProperties}]
 }
 IceDust {
  emitterUpdateScript [
   &EmitterState()
   &SpawnRate(
    SpawnRate=&FloatFromCurve(
     FloatCurve=DataInterface() {propertyValues {Curve {
        defaultValue=1
        keys [
         {arriveTangent=4.1785, interpMode=RCIM_Cubic, leaveTangent=4.1785, tangentMode=RCTM_User, time=0, value=0}
         {arriveTangent=1.1179, interpMode=RCIM_Cubic, leaveTangent=1.1179, tangentMode=RCTM_User, time=0.2042, value=0.6228}
         {arriveTangent=0.0224, interpMode=RCIM_Cubic, leaveTangent=0.0224, tangentMode=RCTM_User, time=1, value=1}
        ]
       }}}
     CurveIndex=$Emitter.NormalizedLoopAge
     Scale Curve=&Lerp_Float(A=13, B=25, Alpha=$User.ParaGlideLeanAlpha)
    )
    Use Spawn Probability=false
   )
   &SpawnRate(
    SpawnGroup=1
    SpawnRate=&FloatFromCurve(
     FloatCurve=DataInterface() {propertyValues {Curve {
        defaultValue=1
        keys [
         {arriveTangent=4.1785, interpMode=RCIM_Cubic, leaveTangent=4.1785, tangentMode=RCTM_User, time=0, value=0}
         {arriveTangent=1.1179, interpMode=RCIM_Cubic, leaveTangent=1.1179, tangentMode=RCTM_User, time=0.2042, value=0.6228}
         {arriveTangent=0.0224, interpMode=RCIM_Cubic, leaveTangent=0.0224, tangentMode=RCTM_User, time=1, value=1}
        ]
       }}}
     CurveIndex=$Emitter.NormalizedLoopAge
     Scale Curve=&Lerp_Float(A=13, B=25, Alpha=$User.ParaGlideLeanAlpha)
    )
    Use Spawn Probability=false
   ) {module=SpawnRate001}
   &SpawnRate(
    SpawnGroup=2
    SpawnRate=&FloatFromCurve(
     FloatCurve=DataInterface() {propertyValues {Curve {
        defaultValue=1
        keys [
         {arriveTangent=4.1785, interpMode=RCIM_Cubic, leaveTangent=4.1785, tangentMode=RCTM_User, time=0, value=0}
         {arriveTangent=1.1179, interpMode=RCIM_Cubic, leaveTangent=1.1179, tangentMode=RCTM_User, time=0.2042, value=0.6228}
         {arriveTangent=0.0224, interpMode=RCIM_Cubic, leaveTangent=0.0224, tangentMode=RCTM_User, time=1, value=1}
        ]
       }}}
     CurveIndex=$Emitter.NormalizedLoopAge
     Scale Curve=7
    )
    Use Spawn Probability=false
   ) {module=SpawnRate002}
   &SpawnRate(
    SpawnGroup=3
    SpawnRate=&FloatFromCurve(
     FloatCurve=DataInterface() {propertyValues {Curve {
        defaultValue=1
        keys [
         {arriveTangent=4.1785, interpMode=RCIM_Cubic, leaveTangent=4.1785, tangentMode=RCTM_User, time=0, value=0}
         {arriveTangent=1.1179, interpMode=RCIM_Cubic, leaveTangent=1.1179, tangentMode=RCTM_User, time=0.2042, value=0.6228}
         {arriveTangent=0.0224, interpMode=RCIM_Cubic, leaveTangent=0.0224, tangentMode=RCTM_User, time=1, value=1}
        ]
       }}}
     CurveIndex=$Emitter.NormalizedLoopAge
     Scale Curve=7
    )
    Use Spawn Probability=false
   ) {module=SpawnRate003}
   &SpawnRate(SpawnGroup=4, SpawnRate=7, Use Spawn Probability=false) {module=SpawnRate004}
   &SpawnRate(SpawnGroup=5, SpawnRate=7, Use Spawn Probability=false) {module=SpawnRate005}
  ]
  particleSpawnScript [
   &InitializeParticle(
    Color=&MergeHelper_ColorSwitch(
     0=&RandomRangeLinearColor(Minimum=<rgba 0.1603 0.1746 0.1927 1>, Maximum=<rgba 0.1905 0.2078 0.24 1>)
     1=&RandomRangeLinearColor(Minimum=<rgba 0.1603 0.1746 0.1927 1>, Maximum=<rgba 0.1905 0.2078 0.24 1>)
     2=&RandomRangeLinearColor(Minimum=<rgba 0.1603 0.1746 0.1927 1>, Maximum=<rgba 0.1905 0.2078 0.24 1>)
     3=&RandomRangeLinearColor(Minimum=<rgba 0.1603 0.1746 0.1927 1>, Maximum=<rgba 0.1905 0.2078 0.24 1>)
     4=&RandomRangeLinearColor(Minimum=<rgba 0.3886 0.705 0.9445 1>, Maximum=<rgba 0.5812 0.7481 0.808 1>)
     5=&RandomRangeLinearColor(Minimum=<rgba 0.3886 0.705 0.9445 1>, Maximum=<rgba 0.5812 0.7481 0.808 1>)
     Index=&SpawnGroup()
     Num=::NewEnumerator4
    )
    Lifetime=&MergeHelper_FloatSwitch(
     0=&RandomRangeFloat(Minimum=0.3, Maximum=0.7)
     1=&RandomRangeFloat(Minimum=0.3, Maximum=0.7)
     2=1.5
     3=1.5
     4=&RandomRangeFloat(Minimum=0.3, Maximum=0.5)
     5=&RandomRangeFloat(Minimum=0.3, Maximum=0.5)
     Index=&SpawnGroup()
     Num=::NewEnumerator4
    )
    Mesh Renderer Array Visibility Mode=::NewEnumerator1
    Mesh Renderer Info=DataInterface() {propertyValues {MeshRenderer=&self:IceDust_0.NiagaraMeshRendererProperties_1}}
    Mesh Scale Max=<0.22 0.22 0.5>
    Mesh Scale Min=<0.1 0.1 0.09>
    Mesh Scale Mode=::NewEnumerator2
    Sprite Size Mode=::NewEnumerator1
    Uniform Sprite Size Max=62
    Uniform Sprite Size Min=50
    UsePositionOffset=false
    Write Material Random=true
   )
   &CameraOffset(Camera Offset Amount=&RandomRangeFloat(Minimum=0, Maximum=4, Evaluation Type=::NewEnumerator1))
   &InitialMeshOrientation(
    Mesh Axis To Orient=<0 0 1>
    Mesh Orientation Mode=::NewEnumerator2
    Orientation Vector=&MergeHelper_VectorSwitch(0=<0 1 -0.18>, 1=<0 -1 0.18>, 2=<0 1 -0.18>, 3=<0 -1 0.18>, Index=&SpawnGroup(), Num=::NewEnumerator2)
    Rotation=&FN_FrontEndVector3Modifier(
     Gameplay=&MergeHelper_VectorSwitch(
      0=&RandomRangeVector(Minimum=<-0.05 -0.05 0>, Maximum=<0.05 0.05 1>)
      1=&RandomRangeVector(Minimum=<-0.05 -0.05 0>, Maximum=<0.05 0.05 1>)
      2=&RandomRangeVector(Minimum=<-0.1 -0.1 0>, Maximum=<0.1 0.1 1>)
      3=&RandomRangeVector(Minimum=<-0.1 -0.1 0>, Maximum=<0.1 0.1 1>)
      Index=&SpawnGroup()
      Num=::NewEnumerator2
     )
     Front End=&MergeHelper_VectorSwitch(
      0=&RandomRangeVector(Minimum=<0 0 0>, Maximum=<1 1 1>)
      1=&RandomRangeVector(Minimum=<0 0 0>, Maximum=<1 1 1>)
      2=&RandomRangeVector(Minimum=<-0.1 -0.1 0>, Maximum=<0.1 0.1 1>)
      3=&RandomRangeVector(Minimum=<-0.1 -0.1 0>, Maximum=<0.1 0.1 1>)
      Index=&SpawnGroup()
      Num=::NewEnumerator2
     )
    )
    Use Rotation Vector=true
   ) {module=InitialMeshOrientation002}
   /SetVariables(Particles.VisibilityTag=&MergeHelper_IntSwitch(0=0, 1=0, 2=0, 3=0, 4=1, 5=1, Index=&SpawnGroup(), Num=::NewEnumerator4)) {isSetParameters=true}
  ]
  particleUpdateScript [
   &ParticleState()
   &SkeletalMeshLocation(
    Bone / Socket Index=&MergeHelper_IntSwitch(0=0, 1=1, 2=0, 3=1, 4=0, 5=1, Index=&SpawnGroup(), Num=::NewEnumerator4)
    Bone Sampling Mode=::NewEnumerator4
    MaskSpawnGroup=false
    Sampled Position Offset=&Add_Vector(
     A=&Lerp_Vector(
      A=&MergeHelper_VectorSwitch(
       0=&RandomRangeVector(Minimum=<-6 3 -4>, Maximum=<6 6 4>)
       1=&RandomRangeVector(Minimum=<-6 -6 -4>, Maximum=<6 -3 4>)
       2=&RandomRangeVector(Minimum=<-6 3 -4>, Maximum=<6 6 4>)
       3=&RandomRangeVector(Minimum=<-6 -6 -4>, Maximum=<6 -3 4>)
       4=<0 0 0>
       5=<0 0 0>
       Index=&SpawnGroup()
       Num=::NewEnumerator4
      )
      B=&MergeHelper_VectorSwitch(
       0=&RandomRangeVector(Minimum=<-10 -5 -5>, Maximum=<-5 5 5>)
       1=&RandomRangeVector(Minimum=<5 -5 -5>, Maximum=<10 5 5>)
       2=&RandomRangeVector(Minimum=<-10 -5 -5>, Maximum=<-5 5 5>)
       3=&RandomRangeVector(Minimum=<5 -5 -5>, Maximum=<10 5 5>)
       4=<0 0 0>
       5=<0 0 0>
       Index=&SpawnGroup()
       Num=::NewEnumerator4
      )
      Alpha=$System.Dive
     )
     B=&MakeVector(
      X=&FN_FrontEndFloatModifier(
       Gameplay=0
       Front End=&FloatFromCurve(
        FloatCurve=DataInterface() {propertyValues {Curve {keys [<at 0.1 0 RCIM_Cubic>,<at 1 11>]}}}
        Scale Curve=&MergeHelper_FloatSwitch(
         0=&Lerp_Float(A=0, B=12, Alpha=$System.Dive)
         1=&Lerp_Float(A=0, B=-12, Alpha=$System.Dive)
         2=0
         3=0
         4=&Lerp_Float(A=0, B=6, Alpha=$System.Dive)
         5=&Lerp_Float(A=0, B=-6, Alpha=$System.Dive)
         Index=&SpawnGroup()
         Num=::NewEnumerator4
        )
       )
      )
      Y=&MergeHelper_FloatSwitch(
       0=&FloatFromCurve(
        FloatCurve=DataInterface() {propertyValues {Curve {keys [<at 0.1 0 RCIM_Cubic>,<at 1 11>]}}}
        Scale Curve=&MergeHelper_FloatSwitch(
         0=&Lerp_Float(A=1, B=0, Alpha=$System.Dive)
         1=&Lerp_Float(A=-1, B=0, Alpha=$System.Dive)
         2=?
         3=?
         Index=&SpawnGroup()
         Num=::NewEnumerator0
        )
       )
       1=&FloatFromCurve(
        FloatCurve=DataInterface() {propertyValues {Curve {keys [<at 0.1 0 RCIM_Cubic>,<at 1 11>]}}}
        Scale Curve=&MergeHelper_FloatSwitch(
         0=&Lerp_Float(A=1, B=0, Alpha=$System.Dive)
         1=&Lerp_Float(A=-1, B=0, Alpha=$System.Dive)
         2=?
         3=?
         Index=&SpawnGroup()
         Num=::NewEnumerator0
        )
       )
       2=0
       3=0
       4=0
       5=0
       Index=&SpawnGroup()
       Num=::NewEnumerator4
      )
      Z=0
     )
    )
    Skeletal Mesh=$System.Skeletal Mesh
   ) {module=SkeletalMeshLocation001}
   &CurlNoiseForce(
    Bias Noise=false
    Noise Frequency=9
    Noise Strength=&MergeHelper_FloatSwitch(0=500, 1=500, 2=0, 3=0, 4=0, 5=0, Index=&SpawnGroup(), Num=::NewEnumerator4)
    Pan Noise=false
   )
   &SolveForcesAndVelocity(Clamp Velocity=false, Limit Acceleration=false)
   &ScaleMeshSize(Scale Factor=&VectorFromFloat(Value=&MergeHelper_FloatSwitch(
      0=&FloatFromCurve(
       FloatCurve=DataInterface() {propertyValues {Curve {keys [
           {arriveTangent=8.0103, interpMode=RCIM_Cubic, leaveTangent=8.0103, tangentMode=RCTM_User, time=0, value=0}
           <at 0.2173 1 RCIM_Cubic>
           {arriveTangent=-4.4721, interpMode=RCIM_Cubic, leaveTangent=1.7195, tangentMode=RCTM_Break, time=1, value=0}
          ]}}}
       Scale Curve=0.9
      )
      1=&FloatFromCurve(
       FloatCurve=DataInterface() {propertyValues {Curve {keys [
           {arriveTangent=8.0103, interpMode=RCIM_Cubic, leaveTangent=8.0103, tangentMode=RCTM_User, time=0, value=0}
           <at 0.2173 1 RCIM_Cubic>
           {arriveTangent=-4.4721, interpMode=RCIM_Cubic, leaveTangent=1.7195, tangentMode=RCTM_Break, time=1, value=0}
          ]}}}
       Scale Curve=0.9
      )
      2=&FloatFromCurve(
       FloatCurve=DataInterface() {propertyValues {Curve {keys [
           {arriveTangent=15.847, interpMode=RCIM_Cubic, leaveTangent=15.847, tangentMode=RCTM_User, time=0, value=0}
           <at 0.1466 1 RCIM_Cubic>
           {arriveTangent=-4.4721, interpMode=RCIM_Cubic, leaveTangent=1.7195, tangentMode=RCTM_Break, time=1, value=0}
          ]}}}
       Scale Curve=1.04
      )
      3=&FloatFromCurve(
       FloatCurve=DataInterface() {propertyValues {Curve {keys [
           {arriveTangent=15.847, interpMode=RCIM_Cubic, leaveTangent=15.847, tangentMode=RCTM_User, time=0, value=0}
           <at 0.1466 1 RCIM_Cubic>
           {arriveTangent=-4.4721, interpMode=RCIM_Cubic, leaveTangent=1.7195, tangentMode=RCTM_Break, time=1, value=0}
          ]}}}
       Scale Curve=1.04
      )
      Index=&SpawnGroup()
      Num=::NewEnumerator2
     )))
   &ScaleColor(
    Scale Alpha=&MergeHelper_FloatSwitch(
     0=&FloatFromCurve(FloatCurve=DataInterface() {propertyValues {Curve {keys [<at 0.5 1 RCIM_Cubic>,<at 1.0011 0.002 RCIM_Cubic>]}}})
     1=&FloatFromCurve(FloatCurve=DataInterface() {propertyValues {Curve {keys [<at 0.5 1 RCIM_Cubic>,<at 1.0011 0.002 RCIM_Cubic>]}}})
     2=&FloatFromCurve(FloatCurve=DataInterface() {propertyValues {Curve {keys [<at 0.5 1 RCIM_Cubic>,<at 1.0011 0.002 RCIM_Cubic>]}}})
     3=&FloatFromCurve(FloatCurve=DataInterface() {propertyValues {Curve {keys [<at 0.5 1 RCIM_Cubic>,<at 1.0011 0.002 RCIM_Cubic>]}}})
     4=&FloatFromCurve(
      FloatCurve=DataInterface() {propertyValues {Curve {keys [<at 0 0 RCIM_Cubic>,<at 0.3153 1 RCIM_Cubic>,<at 1 0>]}}}
      Scale Curve=&Multiply_Float(
       A=&RandomRangeFloat(Minimum=0.2, Maximum=0.4)
       B=&FN_PickFloatDependOnPlatform(TargetPlatform=DataInterface() {propertyValues {Platforms {qualityLevelMask=-29}}}, TargetPlatformActive=0.7, TargetPlatformNoActive=1)
      )
     )
     5=&FloatFromCurve(
      FloatCurve=DataInterface() {propertyValues {Curve {keys [<at 0 0 RCIM_Cubic>,<at 0.3153 1 RCIM_Cubic>,<at 1 0>]}}}
      Scale Curve=&Multiply_Float(
       A=&RandomRangeFloat(Minimum=0.2, Maximum=0.4)
       B=&FN_PickFloatDependOnPlatform(TargetPlatform=DataInterface() {propertyValues {Platforms {qualityLevelMask=-29}}}, TargetPlatformActive=0.7, TargetPlatformNoActive=1)
      )
     )
     Index=&SpawnGroup()
     Num=::NewEnumerator4
    )
    Scale RGB=&MakeVectorFromLinearColorRGB(Color=&MergeHelper_ColorSwitch(
      0=&LerpLinearColors(StartColor=<rgba 0.2083 0.2083 0.2083 1>, EndColor=<rgba 0.3 0.3 0.3 1>, LerpFactor=$System.Dive)
      1=&LerpLinearColors(StartColor=<rgba 0.2083 0.2083 0.2083 1>, EndColor=<rgba 0.3 0.3 0.3 1>, LerpFactor=$System.Dive)
      2=&LerpLinearColors(StartColor=<rgba 0.2083 0.2083 0.2083 1>, EndColor=<rgba 0.3 0.3 0.3 1>, LerpFactor=$System.Dive)
      3=&LerpLinearColors(StartColor=<rgba 0.2083 0.2083 0.2083 1>, EndColor=<rgba 0.3 0.3 0.3 1>, LerpFactor=$System.Dive)
      4=<rgba 1 1 1 1>
      5=<rgba 1 1 1 1>
      Index=&SpawnGroup()
      Num=::NewEnumerator4
     ))
    ScaleA=true
    ScaleRGB=true
   )
   /SetVariables(Particles.MeshOrientation=&LerpQuaternion(
     Quaternion A:Quat4f=$Particles.MeshOrientation
     Quaternion B:Quat4f=&LerpQuaternion(
      Quaternion A:Quat4f=<0 0 0 1>
      Quaternion B:Quat4f=$Particles.MeshOrientation
      Lerp Factor:NiagaraFloat=&FN_FrontEndFloatModifier(Gameplay=0, Front End=1)
      Enable Spherical Linear Interpolation:NiagaraBool=false
     )
     Lerp Factor:NiagaraFloat=$System.Dive
     Enable Spherical Linear Interpolation:NiagaraBool=false
    )) {isSetParameters=true}
   /SetVariables(Particles.Position=&AddVectorToPosition(
     Position=$Particles.Position
     Vector=&MakeVector(
      X=0
      Y=0
      Z=&MergeHelper_FloatSwitch(
       0=&FloatFromCurve(
        FloatCurve=DataInterface() {propertyValues {Curve {keys [<at 0.1 0 RCIM_Cubic>,<at 1 40>]}}}
        Scale Curve=&Lerp_Float(
         A=&RandomRangeFloat(Minimum=2, Maximum=2.5)
         B=&Multiply_Float(
          A=&RandomRangeFloat(Minimum=4, Maximum=6)
          B=&FN_FrontEndFloatModifier(Gameplay=1, Front End=0.22)
         )
         Alpha=$System.Dive
        )
       )
       1=&FloatFromCurve(
        FloatCurve=DataInterface() {propertyValues {Curve {keys [<at 0.1 0 RCIM_Cubic>,<at 1 40>]}}}
        Scale Curve=&Lerp_Float(
         A=&RandomRangeFloat(Minimum=2, Maximum=2.5)
         B=&Multiply_Float(
          A=&RandomRangeFloat(Minimum=4, Maximum=6)
          B=&FN_FrontEndFloatModifier(Gameplay=1, Front End=0.22)
         )
         Alpha=$System.Dive
        )
       )
       2=0
       3=0
       4=&FloatFromCurve(
        FloatCurve=DataInterface() {propertyValues {Curve {keys [<at 0 0 RCIM_Cubic>,<at 1 40>]}}}
        Scale Curve=&Lerp_Float(
         A=2
         B=&Multiply_Float(
          A=&RandomRangeFloat(Minimum=2.5, Maximum=4)
          B=&FN_FrontEndFloatModifier(Gameplay=1, Front End=0.11)
         )
         Alpha=$System.Dive
        )
       )
       5=&FloatFromCurve(
        FloatCurve=DataInterface() {propertyValues {Curve {keys [<at 0 0 RCIM_Cubic>,<at 1 40>]}}}
        Scale Curve=&Lerp_Float(
         A=2
         B=&Multiply_Float(
          A=&RandomRangeFloat(Minimum=2.5, Maximum=4)
          B=&FN_FrontEndFloatModifier(Gameplay=1, Front End=0.11)
         )
         Alpha=$System.Dive
        )
       )
       Index=&SpawnGroup()
       Num=::NewEnumerator4
      )
     )
    )) {isSetParameters=true}
  ]
  rendererClasses [&NiagaraMeshRendererProperties, &NiagaraSpriteRendererProperties]
  renderers [
   {rendererClass=&NiagaraMeshRendererProperties, rendererIndex=0}
   {rendererClass=&NiagaraSpriteRendererProperties, rendererIndex=1}
  ]
 }
 Snowflakes {
  emitterUpdateScript [
   &EmitterState()
   &SpawnRate(
    SpawnRate=&FloatFromCurve(
     FloatCurve=DataInterface() {propertyValues {Curve {
        defaultValue=1
        keys [
         {arriveTangent=4.1785, interpMode=RCIM_Cubic, leaveTangent=4.1785, tangentMode=RCTM_User, time=0, value=0}
         {arriveTangent=1.1179, interpMode=RCIM_Cubic, leaveTangent=1.1179, tangentMode=RCTM_User, time=0.2042, value=0.6228}
         {arriveTangent=0.0224, interpMode=RCIM_Cubic, leaveTangent=0.0224, tangentMode=RCTM_User, time=1, value=1}
        ]
       }}}
     CurveIndex=$Emitter.NormalizedLoopAge
     Scale Curve=&Lerp_Float(A=20, B=30, Alpha=$System.Dive)
    )
    Use Spawn Probability=false
   )
   &SpawnRate(
    SpawnGroup=1
    SpawnRate=&FloatFromCurve(
     FloatCurve=DataInterface() {propertyValues {Curve {
        defaultValue=1
        keys [
         {arriveTangent=4.1785, interpMode=RCIM_Cubic, leaveTangent=4.1785, tangentMode=RCTM_User, time=0, value=0}
         {arriveTangent=1.1179, interpMode=RCIM_Cubic, leaveTangent=1.1179, tangentMode=RCTM_User, time=0.2042, value=0.6228}
         {arriveTangent=0.0224, interpMode=RCIM_Cubic, leaveTangent=0.0224, tangentMode=RCTM_User, time=1, value=1}
        ]
       }}}
     CurveIndex=$Emitter.NormalizedLoopAge
     Scale Curve=&Lerp_Float(A=20, B=30, Alpha=$System.Dive)
    )
    Use Spawn Probability=false
   ) {module=SpawnRate001}
  ]
  particleSpawnScript [
   &InitializeParticle(
    AdjustAlpha=false
    AdjustHue=true
    AdjustSaturation=false
    AdjustValue=false
    Color=<rgba 1.5312 1.7641 2 1>
    Color Mode=::NewEnumerator3
    Lifetime=&MergeHelper_FloatSwitch(
     0=&UniformRangedFloat(Minimum=0.3, Maximum=0.8)
     1=&UniformRangedFloat(Minimum=0.3, Maximum=0.8)
     2=&RandomRangeFloat(Minimum=0.3, Maximum=0.5)
     3=&RandomRangeFloat(Minimum=0.3, Maximum=0.5)
     Index=&SpawnGroup()
     Num=::NewEnumerator2
    )
    Mass Mode=::NewEnumerator1
    Sprite Size Mode=::NewEnumerator1
    Uniform Sprite Size Max=5
    Uniform Sprite Size Min=2
    UsePositionOffset=false
    Write Material Random=true
   ) {module=InitializeParticle001}
   &SkeletalMeshLocation(
    Bone / Socket Index=&MergeHelper_IntSwitch(0=0, 1=1, 2=0, 3=1, 4=?, 5=?, Index=&SpawnGroup(), Num=::NewEnumerator2)
    Bone Sampling Mode=::NewEnumerator4
    MaskSpawnGroup=false
    Sampled Position Offset=&MergeHelper_VectorSwitch(
     0=&RandomRangeVector(Minimum=<-8 -8 -8>, Maximum=<8 8 8>)
     1=&RandomRangeVector(Minimum=<-8 -8 -8>, Maximum=<8 8 8>)
     2=<0 0 0>
     3=<0 0 0>
     Index=&SpawnGroup()
     Num=::NewEnumerator2
    )
    Skeletal Mesh=$System.Skeletal Mesh
   )
   &SubUVAnimation(
    "Sprite Renderer "=DataInterface() {propertyValues {SpriteRenderer=&self:Snowflakes_2.NiagaraSpriteRendererProperties_0}}
    SubUV Animation Mode:ENiagara_SubUVLookupModeV2=::NewEnumerator2
    UseEndFrame=false
    UseStartFrame=false
   )
  ]
  particleUpdateScript [
   &ParticleState()
   &CurlNoiseForce(Noise Strength=160, Pan Noise=true, Pan Noise Field=<1 1 0>)
   &SolveForcesAndVelocity(Clamp Velocity=false, Limit Acceleration=false)
   &SpriteRotationRate(Rotation Rate=300)
   &ScaleColor(
    Scale Alpha=&FloatFromCurve(
     FloatCurve=DataInterface() {propertyValues {Curve {keys [{interpMode=RCIM_Cubic, tangentMode=RCTM_User, time=0.5, value=1}, {interpMode=RCIM_Cubic, tangentMode=RCTM_User, time=1, value=0}]}}}
     Scale Curve=&Multiply_Float(A=1.5, B=&FNTimeOfDayScalar(Night Scale=0.5, Morning Scale=0.9, Evening Scale=0.8))
    )
    Scale RGB=&Multiply_VectorByFloat(Vector=<1 1 1>, Float=&FNTimeOfDayScalar(Night Scale=0.7, Morning Scale=0.9, Evening Scale=0.8))
    ScaleA=true
    ScaleRGB=true
   )
   /SetVariables(Particles.Position=&AddVectorToPosition(
     Position=$Particles.Position
     Vector=&MakeVector(
      X=0
      Y=&FloatFromCurve(
       FloatCurve=DataInterface() {propertyValues {Curve {keys [{arriveTangent=42.1846, interpMode=RCIM_Cubic, time=0, value=0}, <at 1 -40>]}}}
       Scale Curve=&Lerp_Float(
        A=&FN_FrontEndFloatModifier(Gameplay=0, Front End=0.1)
        B=&Multiply_Float(
         A=&RandomRangeFloat(Minimum=0.8, Maximum=1)
         B=&FN_FrontEndFloatModifier(Gameplay=0, Front End=0.5)
        )
        Alpha=$System.Dive
       )
      )
      Z=&FloatFromCurve(
       FloatCurve=DataInterface() {propertyValues {Curve {keys [{arriveTangent=42.1846, interpMode=RCIM_Cubic, leaveTangent=42.1846, tangentMode=RCTM_User, time=0, value=0}, <at 1 40>]}}}
       Scale Curve=&Lerp_Float(
        A=&FN_FrontEndFloatModifier(Gameplay=0.5, Front End=0.35)
        B=&Multiply_Float(
         A=&RandomRangeFloat(Minimum=0.8, Maximum=1)
         B=&FN_FrontEndFloatModifier(Gameplay=1, Front End=0.4)
        )
        Alpha=$System.Dive
       )
      )
     )
    )) {isSetParameters=true}
  ]
  rendererClasses [&NiagaraSpriteRendererProperties]
  renderers [{rendererClass=&NiagaraSpriteRendererProperties}]
 }
 Swirls {
  emitterUpdateScript [
   &EmitterState()
   &SpawnRate(SpawnGroup=1, SpawnRate=&Lerp_Float(A=6, B=8, Alpha=$System.Dive), Use Spawn Probability=false)
   &SpawnRate(SpawnGroup=2, SpawnRate=&Lerp_Float(A=11, B=16, Alpha=$System.Dive), Use Spawn Probability=true) {module=SpawnRate001}
   &InterpolateOverTime(Rate Of Change=50, Target Vector Value:Vector3f=$Engine.Owner.Velocity, Variable Type=::NewEnumerator2)
  ]
  particleSpawnScript [
   &InitializeParticle(
    Lifetime=&MergeHelper_FloatSwitch(
     0=0
     1=&Lerp_Float(
      A=&RandomRangeFloat(Minimum=1, Maximum=3)
      B=&RandomRangeFloat(Minimum=0.7, Maximum=1.2)
      Alpha=$System.Dive
     )
     2=&Lerp_Float(
      A=&RandomRangeFloat(Minimum=0.7, Maximum=0.8)
      B=&RandomRangeFloat(Minimum=0.4, Maximum=0.5)
      Alpha=$System.Dive
     )
     3=0
     4=0
     Index=&SpawnGroup()
     Num=::NewEnumerator3
    )
    Mesh Scale=&RandomRangeVector(Minimum=<2.5 1.8 1.8>, Maximum=<5 2 2>)
    Mesh Scale Mode=::NewEnumerator4
    Sprite Size Mode=::NewEnumerator3
    Uniform Sprite Size=&RandomRangeFloat(Minimum=10, Maximum=25)
    UsePositionOffset=true
    Write Material Random=true
   )
   /SetVariables(Particles.VisibilityTag=&SpawnGroup(Max Spawn Group=9)) {isSetParameters=true, module=SetVariables_#10}
   /SetVariables(Particles.RandomInputPos:NiagaraFloat=&RandomRangeFloat(Minimum=0, Maximum=0)) {isSetParameters=true, module=SetVariables_#11}
   &CameraOffset(Camera Offset Amount=3)
   &InitialMeshOrientation(
    Mesh Orientation Mode=::NewEnumerator7
    Rotation=&RandomRangeVector(Minimum=<0 0 0>, Maximum=<0 0 1>)
    Rotation Coordinate Space=::NewEnumerator2
    Use Rotation Vector=true
   ) {module=InitialMeshOrientation001}
   /SetVariables(Particles.DustDynamic:Vector4f=&MakeVector4(
     X=&RandomRangeFloat(Minimum=-0.05, Maximum=0.05)
     Y=&RandomRangeFloat(Minimum=0, Maximum=32)
     Z=0
     W=0
    )) {isSetParameters=true}
  ]
  particleUpdateScript [
   &ParticleState()
   &ScaleMeshSize(Scale Factor=&Lerp_Vector(
     A=<1 1 1>
     B=&FN_FrontEndVector3Modifier(Gameplay=<2.5 2.5 2.5>, Front End=<1.5 1.5 1.5>)
     Alpha=$System.Dive
    )) {module=ScaleMeshSize001}
   &ScaleSpriteSize(
    Uniform Curve Scale=2.6
    Uniform Curve Sprite Scale=DataInterface() {propertyValues {Curve {keys [
        {arriveTangent=2.2332, interpMode=RCIM_Cubic, leaveTangent=2.2332, tangentMode=RCTM_User, time=0, value=0.6}
        {tangentMode=RCTM_User, time=1, value=1.25}
       ]}}}
   )
   /SetVariables(Particles.Ambient:Vector3f=$System.Ambient) {isSetParameters=true, module=SetVariables_#12}
   &InterpolateOverTime(
    Rate Of Change=3
    Target Float Value:NiagaraFloat=&MakeCustomFloatFromBool(Bool=$System.Timeline001.IsRewinding, True Float=0, False Float=1)
   ) {module=InterpolateOverTime004}
   /SetVariables(Particles.CatchupVector:Vector3f=&Multiply_VectorByFloat(
     Vector=&Multiply_VectorByFloat(
      Vector=&NormalizeVector(Vector To Normalize=$System.SkydiveVelocity, Safe Normalize=true, Safe Fallback Threshold=0, Safe Fallback Vector=<0 0 0.001>)
      Float=-75
     )
     Float=&Lerp_Float(
      A=&FloatFromCurve(
       FloatCurve=DataInterface() {propertyValues {Curve {keys [
           {interpMode=RCIM_Cubic, tangentMode=RCTM_User, time=0, value=0}
           {interpMode=RCIM_Cubic, tangentMode=RCTM_User, time=0.35, value=-0.35}
           {tangentMode=RCTM_User, time=1, value=0}
          ]}}}
       CurveIndex=$System.Timeline001.NormalizedTime
      )
      B=&FloatFromCurve(
       FloatCurve=DataInterface() {propertyValues {Curve {keys [
           {arriveTangent=3.3827, arriveTangentWeight=1.1106, interpMode=RCIM_Cubic, leaveTangentWeight=1.1106, time=0, value=0}
           {arriveTangentWeight=0.5241, interpMode=RCIM_Cubic, leaveTangentWeight=0.1667, time=0.5, value=0.65}
           <at 1 0 RCIM_Cubic>
          ]}}}
       CurveIndex=$System.Timeline001.NormalizedTime
      )
      Alpha=$Particles.InterpolateOverTime004.InterpolateOverTime_Float.Moving Average
     )
    )) {isSetParameters=true, module=SetVariables_#13}
   &SkeletalMeshLocation(
    Bone / Socket Index=4
    Bone Sampling Mode=::NewEnumerator4
    MaskSpawnGroup=false
    Orientation Sampling=::NewEnumerator1
    Sampled Position Offset=&Add_Vector(
     A=$Particles.CatchupVector
     B=&RotateVector(
      Vector To Rotate:Vector3f=&Add_Vector(
       A=$Particles.Ambient
       B=&Add_Vector(
        A=&Add_Vector(
         A=$Particles.ShapeLocation.ShapeVector
         B=&MergeHelper_VectorSwitch(
          0=&Lerp_Vector(A=<0 40 100>, B=<0 80 70>, Alpha=$System.InterpolateOverTime.InterpolateOverTime_Float.Moving Average)
          1=&Lerp_Vector(
           A=&RandomRangeVector(Minimum=<0 40 100>, Maximum=<0 40 110>)
           B=&FN_FrontEndVector3Modifier(
            Gameplay=&RandomRangeVector(Minimum=<0 70 65>, Maximum=<0 90 75>)
            Front End=&RandomRangeVector(Minimum=<0 70 40>, Maximum=<0 90 50>)
           )
           Alpha=$System.InterpolateOverTime.InterpolateOverTime_Float.Moving Average
          )
          2=&Add_Vector(
           A=&Lerp_Vector(A=<0 40 50>, B=<0 80 20>, Alpha=$System.InterpolateOverTime.InterpolateOverTime_Float.Moving Average)
           B=&MakeVector(
            X=0
            Y=0
            Z=&FloatFromCurve(
             FloatCurve=DataInterface() {propertyValues {Curve {keys [<at 0 0>,<at 1 1>]}}}
             Scale Curve=&RandomRangeFloat(Minimum=50, Maximum=80)
            )
           )
          )
          3=<0 0 0>
          4=<0 0 0>
          Index=&SpawnGroup()
          Num=::NewEnumerator3
         )
        )
        B=&Add_Vector(A=$System.OverallTargetOffset, B=$System.FEOffset)
       )
      )
      Yaw:NiagaraFloat=&Waveform(
       "Waveform [1]"=::NewEnumerator12
       "[1] Input Position"=$Particles.RandomInputPos
       "[1] Frequency"=1.25
       "[1] Amplitude Min/Max"=<7 -7>
       "[1] Scale"=0
      )
      Pitch:NiagaraFloat=0
      Roll:NiagaraFloat=0
      Rotation Rate:NiagaraFloat=2
     )
    )
    Skeletal Mesh=$System.Skeletal Mesh
    Use Sampled Mesh Attribute as Ribbon ID=false
   )
   &ScaleMeshSize(Scale Factor=&VectorFromFloat(Value=&FloatFromCurve(
      FloatCurve=DataInterface() {propertyValues {Curve {keys [{arriveTangent=1.0186, interpMode=RCIM_Cubic, leaveTangent=1.0186, tangentMode=RCTM_User, time=0, value=0}, <at 1 1 RCIM_Cubic>]}}}
      CurveIndex=$Emitter.Age
     )))
   /SetVariables(Particles.Wobble:Quat4f=&MultiplyQuaternion(
     Quaternion A=&FindQuatBetweenTwoVectors(
      From Vector=&MakeVector(
       X=&Sine(
        Normalized Angle=&Add_Float(A=$System.FN_Integrator.Value, B=&RandomRangeFloat(Minimum=0.5, Maximum=0.5))
        Period=&MakeCustomFloatFromBool(Bool=&RandomBool(), True Float=1, False Float=-1)
        Scale=0.2
        Bias=0
       )
       Y=&Cosine(
        Normalized Angle=&Add_Float(A=$System.FN_Integrator.Value, B=&RandomRangeFloat(Minimum=0, Maximum=1))
        Period=&MakeCustomFloatFromBool(Bool=&RandomBool(), True Float=1, False Float=-1)
        Scale=0.2
        Bias=0
       )
       Z=&RandomRangeFloat(Minimum=6, Maximum=6)
      )
      To Vector=<0 0 1>
     )
     Quaternion B=$Particles.Initial.MeshOrientation
    )) {isSetParameters=true, module=SetVariables_#14}
   &UpdateMeshOrientation(
    Bank Rate=900000000
    Banking Angle Min / Max:Vector2f=<0.1 -0.1>
    Enabled:NiagaraBool=&SetBoolByFloatComparison(A=&FN_FrontEndFloatModifier(Gameplay=0, Front End=1), B=1, Comparison Type=::NewEnumerator4)
    "Foward Vector (Mesh Space)":Vector3f=&Lerp_Vector(A=<0.5 0 1>, B=<0.5 0 0>, Alpha=$System.InterpolateOverTime.InterpolateOverTime_Float.Moving Average)
    Orientation Method=::NewEnumerator3
    Particle Velocity:Vector3f=&Multiply_VectorByFloat(
     Vector=&Add_Vector(
      A=&ReturnMeshOrientationAxis(Mesh Orientation Axis=::YAxis, Mesh Orientation Quat=$Particles.SkeletalMeshLocation.SampledRotation)
      B=&Lerp_Vector(A=<0 0 -2>, B=<0 0 -0.5>, Alpha=$System.Dive)
     )
     Float=500000
    )
    Rotation Decay Rate In Degrees=900000000
    Up Vector:Vector3f=&Lerp_Vector(
     A=&ReturnMeshOrientationAxis(Mesh Orientation Axis=::YAxis, Mesh Orientation Quat=$Particles.SkeletalMeshLocation.SampledRotation)
     B=&Multiply_VectorByFloat(
      Vector=&ReturnMeshOrientationAxis(Mesh Orientation Axis=::ZAxis, Mesh Orientation Quat=$Particles.SkeletalMeshLocation.SampledRotation)
      Float=1
     )
     Alpha=&FloatFromCurve(FloatCurve=DataInterface() {propertyValues {Curve {keys [<at 0 0.5>,<at 0.25 0.75>]}}}, CurveIndex=$System.Dive)
    )
    Upright Roll Correction Speed=900000000
    Velocity Orientation Convergence Rate=900000000
   ) {module=UpdateMeshOrientation001}
   &UpdateMeshOrientation_RedPepper_TEMP(
    Bank Rate:NiagaraFloat=&Lerp_Float(
     A=&Lerp_Float(
      A=1
      B=30
      Alpha=&Waveform(
       "Waveform [1]"=::NewEnumerator12
       "[1] Input Position"=$Particles.RandomInputPos
       "[1] Phase Scale"=50000
       "[1] Frequency"=0.5
       "[1] Amplitude Min/Max"=<0 1>
      )
     )
     B=900
     Alpha=$System.Dive
    )
    Banking Angle Min / Max:Vector2f=&Lerp_Vector2(A=<-10 10>, B=<-4 4>, Alpha=$System.Dive)
    Enabled:NiagaraBool=&SetBoolByFloatComparison(A=&FN_FrontEndFloatModifier(Gameplay=0, Front End=1), B=0, Comparison Type=::NewEnumerator4)
    "Foward Vector (Mesh Space)":Vector3f=&Lerp_Vector(
     A=<0 0 -1>
     B=&MergeHelper_VectorSwitch(0=<2 0 1>, 1=<-2 0 -1>, 2=<-2 0 -1>, 3=<0 0 0>, 4=<0 0 0>, Index=&SpawnGroup(), Num=::NewEnumerator3)
     Alpha=$System.Dive
    )
    Orientation Method:ENiagara_UpdateMeshOrientationMode=::NewEnumerator3
    Particle Velocity:Vector3f=$Emitter.InterpolateOverTime.InterpolateOverTime_V3_.Moving Average
    "Right Vector (Mesh Space)":Vector3f=&Lerp_Vector(
     A=<0 -1 0>
     B=&MergeHelper_VectorSwitch(0=<0 1 0>, 1=<0 -1 0>, 2=<0 -1 0>, 3=<0 0 0>, 4=<0 0 0>, Index=&SpawnGroup(), Num=::NewEnumerator3)
     Alpha=$System.Dive
    )
    Rotation Decay Rate In Degrees:NiagaraFloat=&Lerp_Float(A=150, B=500, Alpha=$System.Dive)
    Up Vector:Vector3f=&Lerp_Vector(
     A=&ReturnMeshOrientationAxis(Mesh Orientation Axis=::YAxis, Mesh Orientation Quat=$Particles.SkeletalMeshLocation.SampledRotation)
     B=&Multiply_VectorByFloat(
      Vector=&ReturnMeshOrientationAxis(Mesh Orientation Axis=::ZAxis, Mesh Orientation Quat=$Particles.SkeletalMeshLocation.SampledRotation)
      Float=1
     )
     Alpha=$System.Dive
    )
    Upright Roll Correction Speed:NiagaraFloat=&Lerp_Float(A=25, B=90000, Alpha=$System.Dive)
    Velocity Orientation Convergence Rate:NiagaraFloat=&Lerp_Float(
     A=&Lerp_Float(
      A=3
      B=12
      Alpha=&Waveform("Waveform [1]"=::NewEnumerator12, "[1] Input Position"=$Particles.RandomInputPos, "[1] Frequency"=5, "[1] Amplitude Min/Max"=<0 1>)
     )
     B=&Lerp_Float(
      A=20
      B=40
      Alpha=&Waveform("Waveform [1]"=::NewEnumerator12, "[1] Input Position"=$Particles.RandomInputPos, "[1] Frequency"=5, "[1] Amplitude Min/Max"=<0 1>)
     )
     Alpha=$System.Dive
    )
   ) {module=UpdateMeshOrientation_RedPepper_TEMP001}
   &InterpolateOverTime(
    Rate Of Change:NiagaraFloat=&Lerp_Float(
     A=4
     B=2
     Alpha=&FloatFromCurve(
      FloatCurve=DataInterface() {propertyValues {Curve {keys [
          {interpMode=RCIM_Cubic, leaveTangent=0.0851, tangentMode=RCTM_Break, time=0, value=0}
          {interpMode=RCIM_Cubic, tangentMode=RCTM_Break, time=0.222, value=1}
         ]}}}
      CurveIndex=$System.Dive
     )
    )
    Target Quat Value:Quat4f=$Particles.MeshOrientation
    Variable Type=::NewEnumerator5
   ) {module=InterpolateOverTime002}
   /SetVariables(Particles.New Orient:Quat4f=$StackContext.InterpolateOverTime002.InterpolateOverTime_Quaternion.Moving Average) {isSetParameters=true, module=SetVariables_#15}
   /SetVariables(Particles.MeshOrientation=&MultiplyQuaternion(Quaternion A=$StackContext.Wobble, Quaternion B=$Particles.New Orient)) {isSetParameters=true, module=SetVariables_#16}
   &ScaleColor(
    Scale RGB=&MakeVectorFromLinearColorRGB(Color=&LerpLinearColors(StartColor=<rgba 0 1.3565 2 1>, EndColor=<rgba 4 0.479 0 1>, LerpFactor=$System.InterpolateOverTime.InterpolateOverTime_Float.Moving Average))
    ScaleA=true
    ScaleRGB=true
   )
   &DynamicMaterialParameters(
    Index 0 Param 1=&RandomRangeFloat(Minimum=1, Maximum=2)
    Index 0 Param 2=&Multiply_Float(
     A=&RandomRangeFloat(Minimum=1.2, Maximum=4.6)
     B=&Lerp_Float(A=1, B=2.5, Alpha=$System.Dive)
    )
    Index 0 Param 3=&Multiply_Float(
     A=&RandomRangeFloat(Minimum=1.2, Maximum=4.6)
     B=&Lerp_Float(A=1, B=2.5, Alpha=$System.Dive)
    )
    Index 0 Param 4=&RandomRangeFloat(Minimum=1, Maximum=1.6)
    Index 2 Param 1=&FloatFromCurve(FloatCurve=DataInterface() {propertyValues {Curve {keys [<at 0 1 RCIM_Cubic>,<at 0.4739 1 RCIM_Cubic>,<at 1 1>]}}})
    Index 2 Param 2=&FloatFromCurve(FloatCurve=DataInterface() {propertyValues {Curve {keys [<at 0 -1>,<at 1 1>]}}})
    Index 2 Param 3=&RandomRangeFloat(Minimum=10, Maximum=33)
    Index 2 Param 4=1
    Module.Index0Float0Write=true
    Module.Index0Float1Write=true
    Module.Index0Float2Write=true
    Module.Index0Float3Write=true
    Module.Index2Float0Write=true
    Module.Index2Float1Write=true
    Module.Index2Float2Write=true
    Module.Index2Float3Write=false
    Write Parameter Index 2=::NewEnumerator1
   )
   /SetVariables(
    Particles.Decal Color:LinearColor=&LerpLinearColors(
     StartColor=<rgba 0 1.3565 2 3>
     EndColor=<rgba 6 0.9039 0 3>
     LerpFactor=$System.InterpolateOverTime.InterpolateOverTime_Float.Moving Average
    )
    Particles.Decal Offset=15
    Particles.Decal Position:NiagaraPosition=&AddVectorToPosition(
     Position=$Particles.Position
     Vector=&RotateVectorByQuaternion(Vector To Rotate=<-10 0 15>, Quaternion=$Particles.MeshOrientation)
    )
   ) {isSetParameters=true, module=SetVariables_#17}
   /SetVariables(Particles.SwirlsColor:LinearColor=&MultiplyLinearColor(
     Linear Color=&RandomRangeLinearColor(Minimum=<rgba 0.2523 0.5799 1 1>, Maximum=<rgba 1 1 1 1>)
     Scale Factor=&MakeVector4(
      X=1
      Y=1
      Z=1
      W=&Multiply_Float(
       A=&FN_FrontEndFloatModifier(Gameplay=1, Front End=0.6)
       B=&Multiply_Float(A=$System.TODFloat, B=&Lerp_Float(A=1, B=0.45, Alpha=$System.Dive))
      )
     )
    )) {isSetParameters=true}
   /SetVariables(Particles.DustColor:LinearColor=&MakeLinearColorFromVectorAndFloat(
     "Vector (RGB)"=&MakeVectorFromLinearColorRGB(Color=&RandomRangeLinearColor(Minimum=<rgba 0.5349 0.9703 1.3 1>, Maximum=<rgba 0.9351 1.2036 1.3 1>))
     "Float (Alpha)"=&FloatFromCurve(
      FloatCurve=DataInterface() {propertyValues {Curve {keys [<at 0 1 RCIM_Cubic>,<at 1 0>]}}}
      Scale Curve=&Multiply_Float(
       A=&Multiply_Float(
        A=&RandomRangeFloat(Minimum=0.6, Maximum=0.8)
        B=&FN_FrontEndFloatModifier(Gameplay=1, Front End=0.7)
       )
       B=&FN_PickFloatDependOnPlatform(TargetPlatform=DataInterface() {propertyValues {Platforms {qualityLevelMask=-29}}}, TargetPlatformActive=0.5, TargetPlatformNoActive=1)
      )
     )
    )) {isSetParameters=true}
  ]
  rendererClasses [&NiagaraMeshRendererProperties, &NiagaraSpriteRendererProperties]
  renderers [
   {rendererClass=&NiagaraMeshRendererProperties, rendererIndex=0}
   {rendererClass=&NiagaraSpriteRendererProperties, rendererIndex=1}
  ]
 }
 WindTrails001 {
  emitterUpdateScript [
   &EmitterState()
   &SpawnBurst_Instantaneous(Spawn Count=40, Use Spawn Probability=false, UseLoopCountLimit=true)
  ]
  particleSpawnScript [
   &PartitionParticles(
    Normalized Index In Partition Includes One=true
    Particles Per Partition=10
    Partition Mode:ENiagara_PartitionMode=::NewEnumerator1
    Write Partition Index to Ribbon Link Order:ENiagara_AttributeSamplingApplyOutput=::NewEnumerator0
    Write Partition to Ribbon ID=::NewEnumerator0
   )
   &InitializeParticle(
    Color Mode=::NewEnumerator0
    Lifetime=&SelectFloatFromArray(
     Direct Array Index=$Particles.PartitionParticles.Partition
     Float Selection Array=DataInterface() {propertyValues {FloatData [0.3,0.3,0.125,0.125]}}
    )
    Ribbon Width=&SelectFloatFromArray(Direct Array Index=$Particles.PartitionParticles.Partition, Float Selection Array=DataInterface() {propertyValues {FloatData [20,20,13,13]}})
    Ribbon Width Mode=::NewEnumerator1
    Sprite Size Mode=::NewEnumerator3
    Uniform Sprite Size=120
    UsePositionOffset=false
    Write Material Random=true
   )
  ]
  particleUpdateScript [
   &ParticleState(Kill Particles When Lifetime Has Elapsed=false)
   &SkeletalMeshLocation(
    Bone / Socket Index=$Particles.PartitionParticles.Partition
    MaskSpawnGroup=false
    Mesh Sampling Type=::NewEnumerator5
    Skeletal Mesh=$System.Skeletal Mesh
    Socket Sampling Mode=::NewEnumerator5
   )
   /SetVariables(Particles.NewAge:NiagaraFloat=&Multiply_Float(A=$Particles.PartitionParticles.Normalized Index In Partition, B=$Particles.Initial.Lifetime)) {isSetParameters=true}
   /SetVariables(Particles.Velocity=&Multiply_VectorByFloat(
     Vector=&Lerp_Vector(A=$System.FN_Contrail_Setup.Direction Velocity Raw, B=$System.FN_Contrail_Setup.Direction Velocity, Alpha=$Particles.NewAge)
     Float=$Particles.Initial.Lifetime
    )) {isSetParameters=true}
   /SetVariables(Particles.Position=&AddVectorToPosition(
     Position=$Particles.Position
     Vector=&Multiply_VectorByFloat(Vector=$Particles.Velocity, Float=$Particles.PartitionParticles.Normalized Index In Partition)
    )) {isSetParameters=true}
   &ScaleColor(
    Scale Alpha=&Multiply_Float(
     A=&FN_FrontEndFloatModifier(Gameplay=0.4, Front End=0.12)
     B=$System.FN_Contrail_Setup.Normalized Intro Age
    )
    Scale RGB=&MakeVectorFromLinearColorRGB(Color=$System.ToD Color)
    ScaleA=true
    ScaleRGB=true
   )
   /SetVariables(
    Particles.FillTrailColor:LinearColor=&MakeLinearColorFromVectorAndFloat(
     "Vector (RGB)"=&MakeVectorFromLinearColorRGB(Color=$System.ToD Color)
     "Float (Alpha)"=&Multiply_Float(A=0.3, B=&OneMinusFloat(Float=$Particles.PartitionParticles.Normalized Index In Partition))
    )
    Particles.FillTrailDynPar:Vector4f=&MakeVector4(
     X=-2
     Y=0
     Z=1
     W=&SelectFloatFromArray(Direct Array Index=$Particles.PartitionParticles.Partition, Float Selection Array=DataInterface() {propertyValues {FloatData [2,2,1,1]}})
    )
   ) {isSetParameters=true}
  ]
  rendererClasses [&NiagaraRibbonRendererProperties]
  renderers [
   {rendererClass=&NiagaraRibbonRendererProperties}
   {rendererClass=&NiagaraRibbonRendererProperties, rendererIndex=1}
  ]
 }
}
summary {
 systemName=NS_BriskReign_Contrail
 emitters [
  {emitterName=Hand_Glow, rendererClasses [&NiagaraSpriteRendererProperties]}
  {emitterName=IceDust, rendererClasses [&NiagaraMeshRendererProperties, &NiagaraSpriteRendererProperties]}
  {emitterName=Snowflakes, rendererClasses [&NiagaraSpriteRendererProperties]}
  {emitterName=Crystal, rendererClasses [&NiagaraMeshRendererProperties, &NiagaraSpriteRendererProperties]}
  {emitterName=Swirls, rendererClasses [&NiagaraMeshRendererProperties, &NiagaraSpriteRendererProperties]}
  {emitterName=WindTrails001, rendererClasses [&NiagaraRibbonRendererProperties]}
 ]
 userVariables [
  User.bIsFrontEnd {type=NiagaraBool}
  User.bIsFrontEndPreview {type=NiagaraBool}
  User.ParaGlideLeanAlpha {type=NiagaraFloat, defaultValue=0, description=""}
  User.Scale {type=NiagaraFloat, defaultValue=1, description=""}
 ]
}
